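; Guided-motion movement logic for the SinisterMelon air slam fall, overriding the constant-force direction from the owner pawn's control rotation.
; It dots the aim forward vector with the pawn's down vector and range-checks the result against MinDotAllowed.
; An in-range aim is used directly.
; Otherwise, the 2D-normalized aim is rotated around the control right vector by DownwardAttackAngle, forcing a downward attack heading.
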(fn GetDirection_ConstantForce
  (params (GuidedMotionRuntimeData :type Fort%20Movement%20Mode%20Guided%20Motion%20Runtime%20Data%20Object%20Reference))
  (return (select (Math|Float|InRange(Float) (Math|Vector|DotProduct (Math|Vector|GetForwardVector (Pawn|GetControlRotation (Variables|FortMovementMode_BaseExtLogic|GetOwnerPawn) :!class /Script/Engine.Pawn)) (Math|Vector|NegateVector (Transformation|GetActorUpVector (Variables|FortMovementMode_BaseExtLogic|GetOwnerPawn)))) (Variables|Default|GetMinDotAllowed)) (Math|Vector|GetForwardVector (Pawn|GetControlRotation (Variables|FortMovementMode_BaseExtLogic|GetOwnerPawn) :!class /Script/Engine.Pawn)) (Math|Vector|RotateVectorAroundAxis (Math|Vector|Normalize2D(Vector) (Math|Vector|GetForwardVector (Pawn|GetControlRotation (Variables|FortMovementMode_BaseExtLogic|GetOwnerPawn) :!class /Script/Engine.Pawn))) (Variables|Default|GetDownwardAttackAngle) (Math|Vector|GetRightVector (Pawn|GetControlRotation (Variables|FortMovementMode_BaseExtLogic|GetOwnerPawn) :!class /Script/Engine.Pawn))))))
