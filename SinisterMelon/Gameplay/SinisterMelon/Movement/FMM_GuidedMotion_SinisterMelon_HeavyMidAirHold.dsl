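; Guided motion movement mode for SinisterMelon's heavy mid-air hold, defining a constant force along the owner's current velocity direction.
; Force strength is horizontal speed, capped at 1500, scaled by how closely the net-synced aim matches the horizontal travel direction.
; That alignment factor has a 0.15 floor, so aiming away from travel weakens the push without removing it.
; A separate check reports when the server uses client aim, gated by the UseClientAim scalable float.
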(fn GetDirection_ConstantForce
  (params (GuidedMotionRuntimeData :type Fort%20Movement%20Mode%20Guided%20Motion%20Runtime%20Data%20Object%20Reference))
  (return (Math|Vector|Normalize (Variables|/Script/Engine.MovementComponent|GetVelocity (Variables|FortMovementMode_BaseExtLogic|GetOwnerMovementComponent)))))

(fn GetStrength_ConstantForce
  (params (GuidedMotionRuntimeData :type Fort%20Movement%20Mode%20Guided%20Motion%20Runtime%20Data%20Object%20Reference))
  (return (* (Math|Float|Clamp(Float) (Math|Vector|DotProduct (Math|Vector|Normalize2D(Vector) (Math|Vector|GetForwardVector (CharacterMovement|GetNetSyncedBaseAimRotation (Variables|FortMovementMode_BaseExtLogic|GetOwnerMovementComponent)))) (Math|Vector|Normalize2D(Vector) (Transformation|GetVelocity (Variables|FortMovementMode_BaseExtLogic|GetOwnerPawn)))) 0.15) (Math|Float|Clamp(Float) (Math|Vector|VectorLengthXY (Transformation|GetVelocity (Variables|FortMovementMode_BaseExtLogic|GetOwnerPawn))) 0.0 1.5e+03))))

(fn IsServerUsingClientAim
  (params (Guided%20Motion%20Runtime%20Data :type Fort%20Movement%20Mode%20Guided%20Motion%20Runtime%20Data%20Object%20Reference))
  (return (and (Networking|IsServer Guided%20Motion%20Runtime%20Data) (!= 0 (ScalableFloat|GetCachedScalableFloatValue (Variables|Default|GetUseClientAim))))))
