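; MME_SinisterMelon_CustomSprint is a movement extension for SinisterMelon gameplay, named for a custom sprint behavior.
; It declares an activation event that receives the movement runtime context and has no logic attached.
; Its holster check always returns false, so no holster can be applied through this extension.
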(event EventOnExtensionActivate
  (params (RuntimeContext :type MMERuntime%20Context%20Structure%20%28by%20ref%29)))

(fn CanApplyHolster
  (params (HolsterId :type Name%20%28by%20ref%29 :default "None"))
  (return false))
